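# Niagara swing trail effect for the MetroPunk pickaxe, built from three ribbon emitters on shared FN anim-trail modules that follow bone or socket pairs.
# MainTrail (blue) and RedTrail (red, shifting to dark blue) span different socket pairs and overwrite their spawn color during update.
# Tip adds a short accent-colored trail with two ribbon renderers and animated width.
# All three fade alpha over life, taper TipWidth, and use time-of-day color or brightness scaling.
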NiagaraSystem /BRCosmetics/Weapons/FORT_Melee/Pickaxe_MetroPunk/FX/NS_Pickaxe_MetroPunk_SwingTrail
dependencies {
 usedDataInterfaces [&NiagaraDataInterfaceColorCurve, &NiagaraDataInterfaceCurve, &NiagaraDataInterfaceSocketReader]
 usedDynamicInputs []
 usedModules [
  &self:SystemSpawnScript.SystemScriptSource.SystemScriptGraph.NiagaraNodeAssignment_7.SetVariables_#1
  &self:SystemSpawnScript.SystemScriptSource.SystemScriptGraph.NiagaraNodeAssignment_3.SetVariables_#2
  &self:SystemSpawnScript.SystemScriptSource.SystemScriptGraph.NiagaraNodeAssignment_0.SetVariables_#3
  &SystemState
  &self:SystemSpawnScript.SystemScriptSource.SystemScriptGraph.NiagaraNodeAssignment_6.SetVariables_#4
 ]
}
emitters {
 MainTrail {
  emitterSpawnScript [&FN_RibbonIDOffsetUtility()]
  emitterUpdateScript [
   &EmitterState()
   &FN_SpawnAnimTrail()
  ]
  particleSpawnScript [
   &FN_InitializeAnimTrailBoneIndecies(First Bone or Socket Index=1, Second Bone or Socket Index=0)
   /SetVariables(Particles.RibbonWidth=100) {isSetParameters=true, module=SetVariables_#5}
   &FN_InitializeAnimTrail(Color=&MultiplyLinearColor(Linear Color=$System.MainColor, Scale Factor=<1 1 1 1>), Lifetime=0.35)
  ]
  particleUpdateScript [
   &FN_UpdateAnimTrail(Locked Particle RibbonWidth:NiagaraFloat=$Particles.RibbonWidth)
   /SetVariables(Particles.Color=&MultiplyLinearColor(
     Linear Color=&MultiplyLinearColor(
      Linear Color=<rgba 0.0486 0.1363 0.8 1>
      Scale Factor=&MakeVector4_FromLinearColor(LinearColor=&ColorFromCurve(ColorCurve=DataInterface() {propertyValues {
          AlphaCurve {keys [<at 0 1 RCIM_Cubic>,<at 1 0 RCIM_Cubic>]}
          BlueCurve {keys [<at 0.0023 1.25>, {arriveTangent=-0.7517, interpMode=RCIM_Cubic, leaveTangent=-0.7517, time=0.0699, value=1}, <at 1 0.5 RCIM_Cubic>]}
          GreenCurve {keys [<at 0.0023 1.25>, {arriveTangent=-0.7517, interpMode=RCIM_Cubic, leaveTangent=-0.7517, time=0.0699, value=1}, <at 1 0.5 RCIM_Cubic>]}
          RedCurve {keys [<at 0.0023 1.25>, {arriveTangent=-0.7517, interpMode=RCIM_Cubic, leaveTangent=-0.7517, time=0.0699, value=1}, <at 1 0.5 RCIM_Cubic>]}
         }}))
     )
     Scale Factor=&MakeVector4_FromLinearColor(LinearColor=$System.TODColor)
    )) {isSetParameters=true, module=SetVariables_#6}
   &ScaleColor(
    Color Value To Scale=$StackContext.Color
    Scale Alpha=&Multiply_Float(
     A=&FloatFromCurve(
      FloatCurve=DataInterface() {propertyValues {Curve {keys [
          {interpMode=RCIM_Cubic, leaveTangent=-5.8353, tangentMode=RCTM_Break, time=0, value=1.5}
          {arriveTangent=-0.3285, interpMode=RCIM_Cubic, leaveTangent=-0.3285, tangentMode=RCTM_User, time=0.6, value=0.15}
          {arriveTangent=-0.2139, interpMode=RCIM_Cubic, tangentMode=RCTM_Break, time=1, value=0}
         ]}}}
      Scale Curve=2
     )
     B=&FNTimeOfDayScalar(Night Scale=1, Morning Scale=0.5, Evening Scale=0.5)
    )
    Scale RGB=&VectorFromFloat(Value=0.7)
    ScaleA=true
    ScaleRGB=true
   )
   /SetVariables(Particles.TipWidth:NiagaraFloat=&FloatFromCurve(FloatCurve=DataInterface() {propertyValues {Curve {keys [<at 0 1>,<at 1 0>]}}}, Scale Curve=10)) {isSetParameters=true, module=SetVariables_#7}
  ]
  rendererClasses [&NiagaraRibbonRendererProperties]
  renderers [{rendererClass=&NiagaraRibbonRendererProperties}]
 }
 RedTrail {
  emitterSpawnScript [&FN_RibbonIDOffsetUtility()]
  emitterUpdateScript [
   &EmitterState()
   &FN_SpawnAnimTrail()
  ]
  particleSpawnScript [
   &FN_InitializeAnimTrailBoneIndecies(First Bone or Socket Index=1, Second Bone or Socket Index=2)
   /SetVariables(Particles.RibbonWidth=100) {isSetParameters=true, module=SetVariables_#5}
   &FN_InitializeAnimTrail(
    Color=&MultiplyLinearColor(Linear Color=$System.MainColor, Scale Factor=&MakeVector4_FromLinearColor(LinearColor=$System.TODColor))
    Lifetime=0.35
   )
  ]
  particleUpdateScript [
   &FN_UpdateAnimTrail(Locked Particle RibbonWidth:NiagaraFloat=$Particles.RibbonWidth)
   /SetVariables(Particles.Color=&ColorFromCurve(ColorCurve=DataInterface() {propertyValues {
       AlphaCurve {keys [<at 0 1 RCIM_Cubic>,<at 1 0 RCIM_Cubic>]}
       BlueCurve {keys [<at 0.0023 0.0324>, {arriveTangent=0.2104, interpMode=RCIM_Cubic, leaveTangent=0.2104, time=0.0699, value=0.0227}, <at 1 0.2423 RCIM_Cubic>]}
       GreenCurve {keys [<at 0.0023 0.0259>, {arriveTangent=0.0003, interpMode=RCIM_Cubic, leaveTangent=0.0003, time=0.0699, value=0.0181}, <at 1 0.0262 RCIM_Cubic>]}
       RedCurve {keys [<at 0.0023 1>, {arriveTangent=-0.9993, interpMode=RCIM_Cubic, leaveTangent=-0.9993, time=0.0699, value=0.7}, <at 1 0.003 RCIM_Cubic>]}
      }})) {isSetParameters=true, module=SetVariables_#6}
   &ScaleColor(
    Color Value To Scale=$StackContext.Color
    Scale Alpha=&Multiply_Float(
     A=&FloatFromCurve(
      FloatCurve=DataInterface() {propertyValues {Curve {keys [
          {interpMode=RCIM_Cubic, leaveTangent=-5.8353, tangentMode=RCTM_Break, time=0, value=1.5}
          {arriveTangent=-0.3285, interpMode=RCIM_Cubic, leaveTangent=-0.3285, tangentMode=RCTM_User, time=0.6, value=0.15}
          {arriveTangent=-0.2139, interpMode=RCIM_Cubic, tangentMode=RCTM_Break, time=1, value=0}
         ]}}}
      Scale Curve=2
     )
     B=&FNTimeOfDayScalar(Night Scale=1, Morning Scale=0.5, Evening Scale=0.5)
    )
    Scale RGB=&VectorFromFloat(Value=0.7)
    ScaleA=true
    ScaleRGB=true
   )
   /SetVariables(Particles.TipWidth:NiagaraFloat=&FloatFromCurve(FloatCurve=DataInterface() {propertyValues {Curve {keys [<at 0 1>,<at 1 0>]}}}, Scale Curve=10)) {isSetParameters=true, module=SetVariables_#7}
  ]
  rendererClasses [&NiagaraRibbonRendererProperties]
  renderers [{rendererClass=&NiagaraRibbonRendererProperties}]
 }
 Tip {
  emitterSpawnScript [&FN_RibbonIDOffsetUtility()]
  emitterUpdateScript [
   &EmitterState()
   &FN_SpawnAnimTrail()
  ]
  particleSpawnScript [
   &FN_InitializeAnimTrailBoneIndecies(Second Bone or Socket Index=0)
   &FN_InitializeAnimTrail(
    Color=&MultiplyLinearColor(
     Linear Color=$System.AccentColor
     Scale Factor=&MakeVector4_FromLinearColor(LinearColor=&FNTimeOfDayColor(Night Color=<rgba 0.35 0.35 0.35 1>, Morning Color=<rgba 0.75 0.75 0.75 1>, Day Color=<rgba 1 1 1 1>, Evening Color=<rgba 0.7 0.7 0.7 1>))
    )
    Lifetime=0.3
   )
  ]
  particleUpdateScript [
   &FN_UpdateAnimTrail()
   &ScaleColor(
    Scale Alpha=&Multiply_Float(
     A=&FloatFromCurve(
      FloatCurve=DataInterface() {propertyValues {Curve {keys [<at 0 1 RCIM_Cubic>, {arriveTangent=-1, interpMode=RCIM_Cubic, leaveTangent=-1, time=0.4, value=0.2}, <at 1 0 RCIM_Cubic>]}}}
      Scale Curve=$System.SwingGeneralFade
     )
     B=1
    )
    ScaleA=true
    ScaleRGB=true
   )
   /SetVariables(Particles.TipWidth:NiagaraFloat=&FloatFromCurve(FloatCurve=DataInterface() {propertyValues {Curve {keys [<at 0 1>,<at 1 0>]}}}, Scale Curve=5)) {isSetParameters=true, module=SetVariables_#7}
   /SetVariables(Particles.CamFacingWidth:NiagaraFloat=&Multiply_Float(A=$Particles.RibbonWidth, B=0.5)) {isSetParameters=true}
   /SetVariables(Particles.RibbonWidth=&FloatFromCurve(FloatCurve=DataInterface() {propertyValues {Curve {keys [<at 0 0>,<at 0.1 1>,<at 1 0>]}}}, Scale Curve=5)) {isSetParameters=true}
  ]
  rendererClasses [&NiagaraRibbonRendererProperties]
  renderers [
   {rendererClass=&NiagaraRibbonRendererProperties}
   {rendererClass=&NiagaraRibbonRendererProperties, rendererIndex=1}
  ]
 }
}
summary {
 systemName=NS_Pickaxe_MetroPunk_SwingTrail
 emitters [
  {emitterName=MainTrail, rendererClasses [&NiagaraRibbonRendererProperties]}
  {emitterName=Tip, rendererClasses [&NiagaraRibbonRendererProperties]}
  {emitterName=RedTrail, rendererClasses [&NiagaraRibbonRendererProperties]}
 ]
 userVariables [
  User.IsOffhand {type=NiagaraBool}
  User.RibbonIDOffset {type=NiagaraInt32}
  User.SwingSectionIndex {type=NiagaraInt32}
  User.TrailColor {type=LinearColor, description="", defaultValue {struct=LinearColor, value=<rgba 1 0.439 0.1161 0.35>}}
  User.UseBaseEnd {type=NiagaraBool}
  User.VariantIndex {type=NiagaraFloat, defaultValue=0, description=""}
 ]
}
